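; Particle-effect anim notify for wildlife NPC foot turns, which lets the parent notify spawn the effect and then configures it.
; It passes running, left-foot and front-leg flags, plus the mesh's right vector as ForwardVector.
; For wildlife pawn owners, it also passes skeletal mesh index and color variation.
; It sets TurnOffFeathersForChrome when the effect is a feather system and the pawn has the checked gameplay tag.
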(fn Received_Notify
  (params (MeshComp :type Skeletal%20Mesh%20Component%20Object%20Reference) (Animation :type Anim%20Sequence%20Base%20Object%20Reference) (EventReference :type Anim%20Notify%20Event%20Reference%20Structure%20%28by%20ref%29))
  (bind _returnvalue (Components|GetOwner MeshComp))
  (bind (_asnpc_pawn_wildlife_parent _bsuccess) (Utilities|Casting|CastToNPC_Pawn_Wildlife_Parent _returnvalue))
  (bind _returnvalue_2 (|CallParentFunction MeshComp Animation EventReference))
  (bind _returnvalue_3 (AnimNotify|GetSpawnedEffect))
  (Utilities|IsValid _returnvalue_3
    (:Is%20Valid
      (Effects|Components|ParticleSystem|SetBoolParameter _returnvalue_3 "bIsRunning" (Variables|Default|Getis%20Running%20?))
      (Effects|Components|ParticleSystem|SetBoolParameter _returnvalue_3 "bIsLeftFoot" (Variables|Default|GetLeft%20Foot))
      (Effects|Components|ParticleSystem|SetBoolParameter _returnvalue_3 "bIsFrontLeg" (Variables|Default|Getis%20Front%20Leg))
      (Effects|Components|ParticleSystem|SetVectorParameter _returnvalue_3 "ForwardVector" (Transformation|GetRightVector MeshComp))
      (if _bsuccess
        (Effects|Components|ParticleSystem|SetIntParameter _returnvalue_3 "SkeletalMeshIndex" (Math|Float|Truncate (Variables|/Game/Athena/AI/Pawns/NPC/Parents/NPC_Pawn_Parent.NPC_Pawn_Parent_C|GetNPC%20Skeletal%20Mesh%20Index _asnpc_pawn_wildlife_parent)))
        (Effects|Components|ParticleSystem|SetFloatParameter _returnvalue_3 "ColorCurveDataValue" (Variables|/Game/Athena/AI/Pawns/NPC/Parents/NPC_Pawn_Wildlife_Parent.NPC_Pawn_Wildlife_Parent_C|GetColorVariation _asnpc_pawn_wildlife_parent))
        (Effects|Components|ParticleSystem|SetBoolParameter _returnvalue_3 "TurnOffFeathersForChrome" (Math|Boolean|ANDBoolean (Utilities|IsValid (Effects|Components|ParticleSystem|GetFXSystemAsset _returnvalue_3) :!class /Script/Engine.KismetSystemLibrary) (or (== (Effects|Components|ParticleSystem|GetFXSystemAsset _returnvalue_3) (Utilities|ResolveSoftReference (LocalVariables|GetNSFeatherBurst))) (== (Effects|Components|ParticleSystem|GetFXSystemAsset _returnvalue_3) (Utilities|ResolveSoftReference (LocalVariables|GetNSRunFeathers)))) (GameplayTags|HasMatchingGameplayTag _asnpc_pawn_wildlife_parent (Variables|Default|GetTag%20to%20Check))))
        (return _returnvalue_2)))))
