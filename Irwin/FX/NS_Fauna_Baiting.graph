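# Niagara effect for fauna baiting that pops question-mark sprites over a soft, bluish smoke puff.
# Qestion_Marks fires four staggered bursts; each spawn group picks a slot on a camera-facing ring, with squash-and-stretch scale curves and a damped sine wobble.
# SMoke bursts rotating SubUV sprites that fade in and out.
# Three radial sense-wave emitters are present but disabled; User.Camera Offset pulls the question marks toward the camera.
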NiagaraSystem /Irwin/FX/NS_Fauna_Baiting
dependencies {
 usedDataInterfaces [
  &NiagaraDataInterfaceArrayFloat
  &NiagaraDataInterfaceCurve
  &NiagaraDataInterfaceCamera
  &NiagaraDataInterfaceVector2DCurve
 ]
 usedDynamicInputs []
 usedModules [&SystemState]
}
emitters {
 Qestion_Marks {
  emitterUpdateScript [
   &EmitterState()
   &SpawnRate(SpawnRate=3, Use Spawn Probability=false) {enabled=false}
   /SetVariables(Emitter.Cam Pos:NiagaraPosition=&ReturnCameraVector(Camera Property=::NewEnumerator0, Coordinate Space=::Simulation)) {isSetParameters=true}
   /SetVariables(Emitter.Master SPeed=0.1) {isSetParameters=true}
   &SpawnBurst_Instantaneous(Use Spawn Probability=false, UseLoopCountLimit=false)
   &SpawnBurst_Instantaneous(
    Spawn Group=1
    Spawn Time=&Multiply_Float(A=1, B=$Emitter.Master SPeed)
    Use Spawn Probability=false
    UseLoopCountLimit=false
   ) {module=SpawnBurst_Instantaneous003}
   &SpawnBurst_Instantaneous(
    Spawn Group=2
    Spawn Time=&Multiply_Float(A=2, B=$Emitter.Master SPeed)
    Use Spawn Probability=false
    UseLoopCountLimit=false
   ) {module=SpawnBurst_Instantaneous002}
   &SpawnBurst_Instantaneous(
    Spawn Group=3
    Spawn Time=&Multiply_Float(A=3, B=$Emitter.Master SPeed)
    Use Spawn Probability=false
    UseLoopCountLimit=false
   ) {module=SpawnBurst_Instantaneous001}
  ]
  particleSpawnScript [
   &InitializeParticle(
    Color Maximum=<rgba 0.744 0.8554 1 0.5>
    Color Minimum=<rgba 0.5 0.7176 1 0.5>
    Color Mode=::NewEnumerator2
    Lifetime Max=1.75
    Lifetime Min=1.4
    Lifetime Mode=::NewEnumerator1
    Mass Mode=::NewEnumerator3
    Sprite Size Mode=::NewEnumerator1
    Uniform Sprite Size Max=14.5
    Uniform Sprite Size Min=10
    UsePositionOffset=false
    Write Material Random=true
   )
   &CameraOffset(Camera Offset Amount=$User.Camera Offset)
  ]
  particleUpdateScript [
   &ParticleState()
   &ShapeLocation(
    MaskSpawnGroup=false
    Non Uniform Scale=<1.25 1 1>
    Override Seed=false
    Radius Position:NiagaraFloat=&RandomRangeFloat(Minimum=0.5, Maximum=1)
    Ring / Disc Distribution Mode:ENiagaraSphereDistributionMode=::NewEnumerator1
    Ring Radius=38.2
    Rotation Angle:NiagaraFloat=&self:ScratchDynamicInput(
     Degree Offset:NiagaraFloat=90
     New Position:NiagaraPosition=&ReturnCameraVector(Camera Property=::NewEnumerator4, Coordinate Space=::Simulation, Origin Position=/Niagara/DynamicInputs/Helpers/SimulationPosition())
    )
    Rotation Angle Type:ENiagara_AngleInput=::NewEnumerator0
    Rotation Axis=&ReturnCameraVector(Camera Property=::NewEnumerator4, Coordinate Space=::Simulation, Origin Position=&SimulationPosition())
    Rotation Coordinate Space=::World
    Rotation Mode=::NewEnumerator1
    Shape Primitive=::NewEnumerator4
    U Position=&RandomRangeFloat(Minimum=0, Maximum=0.5)
    UseOwnerScale=true
   ) {enabled=false}
   &ShapeLocation(
    MaskSpawnGroup=false
    Non Uniform Scale=<1.25 1 1>
    Override Seed=false
    Radius Position:NiagaraFloat=&RandomRangeFloat(Minimum=0.75, Maximum=1)
    Ring / Disc Distribution Mode:ENiagaraSphereDistributionMode=::NewEnumerator1
    Ring Radius=27.5
    Rotation Angle:NiagaraFloat=&self:ScratchDynamicInput(
     Degree Offset:NiagaraFloat=90
     New Position:NiagaraPosition=&ReturnCameraVector(Camera Property=::NewEnumerator4, Coordinate Space=::Simulation, Origin Position=&SimulationPosition())
    )
    Rotation Angle Type:ENiagara_AngleInput=::NewEnumerator0
    Rotation Axis=&ReturnCameraVector(Camera Property=::NewEnumerator4, Coordinate Space=::Simulation, Origin Position=&SimulationPosition())
    Rotation Coordinate Space=::World
    Rotation Mode=::NewEnumerator1
    Shape Primitive=::NewEnumerator4
    U Position=&SelectFloatFromArray(
     Direct Array Index=&SpawnGroup()
     Float Selection Array=DataInterface() {propertyValues {FloatData [0.32,0.21,0.45,0.05]}}
    )
    UseOwnerScale=true
   ) {module=ShapeLocation001}
   &ScaleColor(
    Scale Alpha=&FloatFromCurve(FloatCurve=DataInterface() {propertyValues {Curve {keys [<at 0.75 1 RCIM_Cubic>,<at 1.0011 0.002 RCIM_Cubic>]}}})
    ScaleA=true
    ScaleRGB=true
   )
   &SolveForcesAndVelocity(Clamp Velocity=false, Limit Acceleration=false)
   &ScaleSpriteSize(
    Non-Uniform Curve Sprite Scale=DataInterface() {propertyValues {
      XCurve {keys [
        {interpMode=RCIM_Cubic, leaveTangentWeight=0.0366, tangentMode=RCTM_User, time=0, value=0}
        {arriveTangentWeight=0.1383, interpMode=RCIM_Cubic, leaveTangentWeight=0.2074, tangentMode=RCTM_User, time=0.0919, value=1.5}
        {arriveTangentWeight=0.0646, interpMode=RCIM_Cubic, leaveTangentWeight=0.2521, tangentMode=RCTM_User, time=0.2568, value=1}
        {arriveTangentWeight=0.3594, interpMode=RCIM_Cubic, leaveTangentWeight=0.0559, tangentMode=RCTM_User, time=0.9, value=1}
        <at 1 0 RCIM_Cubic>
       ]}
      YCurve {keys [
        {interpMode=RCIM_Cubic, leaveTangentWeight=0.0183, tangentMode=RCTM_User, time=0, value=0}
        {arriveTangentWeight=0.0853, interpMode=RCIM_Cubic, leaveTangentWeight=0.1706, tangentMode=RCTM_User, time=0.037, value=2}
        {arriveTangentWeight=0.1383, interpMode=RCIM_Cubic, leaveTangentWeight=0.2074, tangentMode=RCTM_User, time=0.1469, value=0.75}
        {arriveTangentWeight=0.0579, interpMode=RCIM_Cubic, leaveTangentWeight=0.2066, tangentMode=RCTM_User, time=0.3118, value=1}
        {arriveTangentWeight=0.3458, interpMode=RCIM_Cubic, leaveTangentWeight=0.0588, tangentMode=RCTM_User, time=0.9, value=1}
        <at 1 0 RCIM_Cubic>
       ]}
     }}
    Scale Sprite Size Mode=::NewEnumerator3
   )
   /SetVariables(Particles.SpriteRotation:NiagaraFloat=&Add_Float(
     A=&Multiply_Float(A=1, B=1)
     B=&Multiply_Float(
      A=&FloatFromCurve(FloatCurve=DataInterface() {propertyValues {Curve {keys [<at 0 1>,<at 0.45 0.25>]}}})
      B=&Lerp_Float(
       A=-15
       B=15
       Alpha=&Sine(Normalized Angle=&Add_Float(A=$Emitter.Age, B=$Particles.MaterialRandom), Period=1.25, Scale=0.5, Bias=0.5)
      )
     )
    )) {isSetParameters=true}
  ]
  rendererClasses [&NiagaraSpriteRendererProperties]
  renderers [{rendererClass=&NiagaraSpriteRendererProperties}]
 }
 SMoke {
  emitterUpdateScript [
   &EmitterState()
   &SpawnRate(SpawnRate=5, Use Spawn Probability=false) {enabled=false}
   &SpawnBurst_Instantaneous(Spawn Count=10, Use Spawn Probability=false, UseLoopCountLimit=false)
  ]
  particleSpawnScript [
   &InitializeParticle(
    Color=<rgba 1.25 2.8821 5 0.01>
    Lifetime Max=2.5
    Lifetime Min=2.5
    Lifetime Mode=::NewEnumerator1
    Mass Mode=::NewEnumerator3
    Sprite Rotation Mode=::NewEnumerator2
    Sprite Size Mode=::NewEnumerator1
    Uniform Sprite Size Max=70
    Uniform Sprite Size Min=50
    UsePositionOffset=true
    Write Material Random=true
   )
   &CameraOffset(Camera Offset Amount=&RandomRangeFloat(Minimum=20, Maximum=40))
   &ShapeLocation(
    Hemisphere Distribution=<1 0.85>
    MaskSpawnGroup=false
    Override Seed=false
    Rotation Mode=::NewEnumerator2
    Sphere Radius=25
    Sphere Surface Distribution=1
    UseOwnerScale=true
    Yaw / Pitch / Roll=<0 0 -0.25>
   )
   &DynamicMaterialParameters(
    Index 0 Param 1=&FloatFromCurve(FloatCurve=DataInterface() {propertyValues {Curve {keys [<at 0 1>,<at 1 0>]}}})
    Index 0 Param 2=1
    Index 0 Param 3=1
    Index 0 Param 4=1
    Module.Param0WriteEnabled=true
    Module.Param1WriteEnabled=true
    Module.Param2WriteEnabled=true
    Module.Param3WriteEnabled=true
   )
  ]
  particleUpdateScript [
   &SubUVAnimation(
    Sub UV Play Rate=0.25
    SubUV Animation Mode:ENiagara_SubUVLookupMode=::NewEnumerator3
    SubUV Start Time Offset:NiagaraFloat=&RandomRangeFloat(Minimum=0, Maximum=100)
   )
   &ParticleState()
   &ScaleColor(
    Scale Alpha=&FloatFromCurve(
     FloatCurve=DataInterface() {propertyValues {Curve {keys [
         {interpMode=RCIM_Cubic, tangentMode=RCTM_User, time=0, value=0}
         {interpMode=RCIM_Cubic, tangentMode=RCTM_User, time=0.5, value=1}
         {tangentMode=RCTM_User, time=1, value=0}
        ]}}}
     Scale Curve=0.35
    )
    ScaleA=true
    ScaleRGB=true
   )
   &SpriteRotationRate(Rotation Rate=&RandomRangeFloat(Minimum=60, Maximum=60))
  ]
  rendererClasses [&NiagaraSpriteRendererProperties]
  renderers [{rendererClass=&NiagaraSpriteRendererProperties}]
 }
 Sense_Wave {
  bEnabled=false
  emitterUpdateScript [
   &EmitterState(Life Cycle Mode=::NewEnumerator1, Loop Behavior=::NewEnumerator1, Loop Delay=0.2, Loop Duration=1, UseLoopDelay=false)
   &SpawnBurst_Instantaneous(Spawn Count=1, Use Spawn Probability=false)
  ]
  particleSpawnScript [&InitializeParticle(
    Color Mode=::NewEnumerator0
    Lifetime=1.6
    Sprite Rotation Mode=::NewEnumerator0
    Sprite Size Mode=::NewEnumerator3
    Uniform Sprite Size=&Multiply_Float(
     A=50
     B=&SelectFloatFromArray(
      Direct Array Index=&MakeIntFromFloat(Float=0)
      Float Selection Array=DataInterface() {propertyValues {FloatData [4.25,3.75,4.5,4.125]}}
     )
    )
    UsePositionOffset=false
    Write Material Random=true
   )]
  particleUpdateScript [
   &ParticleState()
   &Color(
    Color=<rgba 2.8302 3.762 3.8 1>
    Scale Alpha=&FloatFromCurve(
     FloatCurve=DataInterface() {propertyValues {Curve {keys [
         {interpMode=RCIM_Cubic, tangentMode=RCTM_Break, time=0, value=0}
         {interpMode=RCIM_Cubic, tangentMode=RCTM_Break, time=0.2, value=1}
         {interpMode=RCIM_Cubic, tangentMode=RCTM_Break, time=1, value=0}
        ]}}}
     Scale Curve=&Multiply_Float(A=1.25, B=$User.Sense Opacity Multiplier)
    )
   )
   &ScaleSpriteSize(Scale Factor=&Vector2DFromFloat(Value=&FloatFromCurve(FloatCurve=DataInterface() {propertyValues {Curve {keys [<at 0 0.5>,<at 1 1>]}}})))
   &CameraOffset(Camera Offset Amount=$User.Camera Offset)
   &FN_Radial_Material(Radial Mask Offset X=1, Radial Mask Offset Y=0.145, Radial Mask Power=3, Radial Mask Tile U=4, Sprite Extrusion Speed=1)
  ]
  rendererClasses [&NiagaraSpriteRendererProperties]
  renderers [{rendererClass=&NiagaraSpriteRendererProperties}]
 }
 Wavy {
  bEnabled=false
  emitterUpdateScript [
   &EmitterState(Life Cycle Mode=::NewEnumerator1, Loop Behavior=::NewEnumerator1, Loop Duration=1, UseLoopDelay=false)
   &SpawnBurst_Instantaneous(Spawn Count=1, Use Spawn Probability=false)
  ]
  particleSpawnScript [&InitializeParticle(
    Color Mode=::NewEnumerator0
    Lifetime=1.6
    Sprite Size Mode=::NewEnumerator3
    Uniform Sprite Size=&Multiply_Float(
     A=50
     B=&SelectFloatFromArray(
      Direct Array Index=&MakeIntFromFloat(Float=0)
      Float Selection Array=DataInterface() {propertyValues {FloatData [4.25,3.75,4.5,4.125]}}
     )
    )
    UsePositionOffset=false
    Write Material Random=true
   )]
  particleUpdateScript [
   &ParticleState()
   &Color(
    Color=<rgba 0.7964 2.6259 5 1>
    Scale Alpha=&FloatFromCurve(
     FloatCurve=DataInterface() {propertyValues {Curve {keys [
         {interpMode=RCIM_Cubic, tangentMode=RCTM_Break, time=0, value=0}
         {interpMode=RCIM_Cubic, tangentMode=RCTM_Break, time=0.2, value=1}
         {interpMode=RCIM_Cubic, tangentMode=RCTM_Break, time=1, value=0}
        ]}}}
     Scale Curve=&Multiply_Float(A=1, B=$User.Sense Opacity Multiplier)
    )
   )
   &FN_Radial_Material(Radial Mask Offset X=1, Radial Mask Offset Y=0.145, Radial Mask Power=3, Radial Mask Tile U=4, Sprite Extrusion Speed=1)
   &ScaleSpriteSize(Scale Factor=&Vector2DFromFloat(Value=&FloatFromCurve(FloatCurve=DataInterface() {propertyValues {Curve {keys [<at 0 0.5>,<at 1 1>]}}})))
   &CameraOffset(Camera Offset Amount=$User.Camera Offset)
  ]
  rendererClasses [&NiagaraSpriteRendererProperties]
  renderers [{rendererClass=&NiagaraSpriteRendererProperties}]
 }
 Wavy_02 {
  bEnabled=false
  emitterUpdateScript [
   &EmitterState(Life Cycle Mode=::NewEnumerator1, Loop Behavior=::NewEnumerator1, Loop Duration=1, UseLoopDelay=false)
   &SpawnBurst_Instantaneous(Spawn Count=1, Use Spawn Probability=false)
  ]
  particleSpawnScript [&InitializeParticle(
    Color Mode=::NewEnumerator0
    Lifetime=1.6
    Sprite Size Mode=::NewEnumerator3
    Uniform Sprite Size=&Multiply_Float(
     A=42
     B=&SelectFloatFromArray(
      Direct Array Index=&MakeIntFromFloat(Float=0)
      Float Selection Array=DataInterface() {propertyValues {FloatData [4.25,3.75,4.5,4.125]}}
     )
    )
    UsePositionOffset=false
    Write Material Random=true
   )]
  particleUpdateScript [
   &ParticleState()
   &Color(
    Color=<rgba 0 3.716 5 1>
    Scale Alpha=&FloatFromCurve(
     FloatCurve=DataInterface() {propertyValues {Curve {keys [
         {interpMode=RCIM_Cubic, tangentMode=RCTM_Break, time=0, value=0}
         {interpMode=RCIM_Cubic, tangentMode=RCTM_Break, time=0.2, value=1}
         {interpMode=RCIM_Cubic, tangentMode=RCTM_Break, time=1, value=0}
        ]}}}
     Scale Curve=&Multiply_Float(A=0.7, B=$User.Sense Opacity Multiplier)
    )
   )
   &FN_Radial_Material(Radial Mask Offset X=1, Radial Mask Offset Y=0.145, Radial Mask Power=3, Radial Mask Tile U=4, Sprite Extrusion Speed=1)
   &ScaleSpriteSize(Scale Factor=&Vector2DFromFloat(Value=&FloatFromCurve(FloatCurve=DataInterface() {propertyValues {Curve {keys [<at 0 0.5>,<at 1 1>]}}})))
   &CameraOffset(Camera Offset Amount=$User.Camera Offset)
  ]
  rendererClasses [&NiagaraSpriteRendererProperties]
  renderers [{rendererClass=&NiagaraSpriteRendererProperties}]
 }
}
summary {
 systemName=NS_Fauna_Baiting
 emitters [
  {bEnabled=false, emitterName=Wavy, rendererClasses [&NiagaraSpriteRendererProperties]}
  {bEnabled=false, emitterName=Wavy_02, rendererClasses [&NiagaraSpriteRendererProperties]}
  {bEnabled=false, emitterName=Sense_Wave, rendererClasses [&NiagaraSpriteRendererProperties]}
  {emitterName=Qestion_Marks, rendererClasses [&NiagaraSpriteRendererProperties]}
  {emitterName=SMoke, rendererClasses [&NiagaraSpriteRendererProperties]}
 ]
 userVariables [
  User.Camera Offset {type=NiagaraFloat, defaultValue=30, description=""}
  User.Sense Opacity Multiplier {type=NiagaraFloat, defaultValue=0.65, description=""}
  User.Sense Size Multiplier {type=NiagaraFloat, defaultValue=1, description=""}
 ]
}
